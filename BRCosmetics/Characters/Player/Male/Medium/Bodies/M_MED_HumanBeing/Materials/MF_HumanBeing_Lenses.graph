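# Material function for the HumanBeing character's lenses, adding animated concentric rings over a two-color split.
# A step on UV0's U axis picks the lens color and offsets the ring phase for one half.
# Powered radial distance from UV1's center, scrolled by time, passes through Frac and Step to form hard-edged bands.
# These bands tint color and vary roughness, and an Alpha input blends BaseColor, Roughness and Emissive into the incoming material attributes.
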MaterialFunction /BRCosmetics/Characters/Player/Male/Medium/Bodies/M_MED_HumanBeing/Materials/MF_HumanBeing_Lenses
graph {nodes [
  n1 = NamedRerouteUsage()
  n2 = Step(Y=ComponentMask(None=TextureCoordinate()) {r=true}) {constX=0.5}
  n3 = LinearInterpolate(
   A=VectorParameter() {group=Humanbeing Lens, parameterName=Color_1, defaultValue=<rgba 0.521 0 0.0168 1>}.RGB
   B=VectorParameter() {group=Humanbeing Lens, parameterName=Color_2, defaultValue=<rgba 0.0003 0.2384 0.3095 1>}.RGB
   Alpha=@n2
  )
  n4 = OneMinus(None=@n2)
  n5 = TextureCoordinate() {coordinateIndex=1}
  n6 = Power(
   Base=MaterialFunctionCall(Vector 2=Subtract(A=@n5) {constB=0.5}) {materialFunction=/Engine/Functions/Engine_MaterialFunctions02/Utility/VectorLength}.V2 Length
   Exp=ScalarParameter() {defaultValue=1.4, group=Humanbeing Lens, parameterName=RingsStartPower}
  ) {constExponent=1.25}
  n7 = Multiply(A=@n2, B=ScalarParameter() {defaultValue=0.05, group=Humanbeing Lens, parameterName=RingsStartOffset})
  n8 = Multiply(A=@n2, B=@n6)
  n9 = Add(
   A=Add(A=Multiply(A=@n4, B=@n6), B=Add(A=@n7, B=@n8) {constB=0})
   B=Multiply(A=Time(), B=ScalarParameter() {defaultValue=-0.18, group=Humanbeing Lens, parameterName=RingsSpeed})
  ) {constB=0}
  n10 = ScalarParameter() {defaultValue=6, group=Humanbeing Lens, parameterName=RingsAmount}
  n11 = Step(Y=Frac(None=Multiply(A=@n9, B=@n10))) {constX=0.5}
  n12 = LinearInterpolate(
   A=@n3
   B=LinearInterpolate(
    A=@n3
    B=Constant3Vector() {constant=<rgba 1 1 1 0>}
    Alpha=ScalarParameter() {defaultValue=0.1, group=Humanbeing Lens, parameterName=RingsColorBias}
   )
   Alpha=NamedRerouteDeclaration(None=Saturate(None=@n11)) {name=GlassesRingsMask, nodeColor=<rgba 0.8047 0 1 1>}
  )
  n13 = NamedRerouteUsage()
  n14 = NamedRerouteUsage()
  n15 = LinearInterpolate(
   A=ScalarParameter() {defaultValue=0.4, group=Humanbeing Lens, parameterName=RingsRoughness_02}
   B=ScalarParameter() {defaultValue=0.3, group=Humanbeing Lens, parameterName=RingsRoughness_01}
   Alpha=NamedRerouteUsage()
  )
  n16 = NamedRerouteUsage()
  n17 = Multiply(A=@n12, B=ScalarParameter() {defaultValue=0.1, group=Humanbeing Lens, parameterName=EmissiveIntensity})
  FunctionOutput(None=SetMaterialAttributes!(
    MaterialAttributes=NamedRerouteUsage()
    BaseColor=LinearInterpolate(A=@n1, B=@n12, Alpha=@n13)
    Roughness=LinearInterpolate(A=@n14, B=@n15, Alpha=@n13)
    EmissiveColor=LinearInterpolate(A=@n16, B=@n17, Alpha=@n13)
   ) {attributeSetTypes [#1, #2, #3]})
  n18 = GetMaterialAttributes(None=FunctionInput() {inputName=Material, inputType=FunctionInput_MaterialAttributes}) {attributeGetTypes [#1, #2, #3]}
  NamedRerouteDeclaration(None=@n18.EmissiveColor) {name=EmissiveColor_IN, nodeColor=<rgba 1 0 0.1641 1>}
  NamedRerouteDeclaration(None=@n18.MaterialAttributes) {name=MA_IN, nodeColor=<rgba 0.125 0 1 1>}
  NamedRerouteDeclaration(None=FunctionInput() {inputName=Alpha, inputType=FunctionInput_Scalar}) {name=Alpha, nodeColor=<rgba 0.3125 1 0 1>}
  NamedRerouteDeclaration(None=@n18.BaseColor) {name=BaseColor_IN, nodeColor=<rgba 0 1 0.9766 1>}
  NamedRerouteDeclaration(None=@n18.Roughness) {name=Roughness_IN, nodeColor=<rgba 0.2188 0 1 1>}
 ]}
