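# Skinned character material for the face accessory of the male medium HumanBeing body, layering masked surface effects between the shared character PreFX and PostFX functions.
# Channels of the SkinFX mask drive optional cloth fuzz, silk and thin film layers, each behind a static switch.
# A HumanBeing-specific lens function is always applied, masked by its own channel parameter.
# The result passes through PostFX, then a cinematic post-FX function, as material attributes.
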Material /BRCosmetics/Characters/Player/Male/Medium/Bodies/M_MED_HumanBeing/Materials/M_MED_HumanBeing_FaceAcc_FX
graph {
 properties {
  bUseMaterialAttributes=true
  bUsedWithMeshDeformer=true
  bUsedWithMorphTargets=true
  bUsedWithSkeletalMesh=true
  pixelDepthOffsetMode=PDOM_Legacy
  usedShadingModels=Default Lit
 }
 nodes [
  n1 = MaterialFunctionCall() {materialFunction=&BaseFN_CharacterFunction_PreFX}
  n2 = TextureObjectParameter() {
   group=Skin FX
   parameterName=SkinFX_Mask
   samplerType=SAMPLERTYPE_Masks
   texture=/BRCosmetics/Characters/Player/Male/Medium/Bodies/M_MED_HumanBeing/Textures/T_Character_HumanBeing_Texture_FaceAcc_FX
  }
  n3 = TextureSample(Tex=@n2)
  n4 = StaticSwitchParameter(
   True=MaterialFunctionCall!(
    Alpha=StaticComponentMaskParameter(None=@n3.RGBA) {defaultR=true, group=Skin FX, parameterName=Cloth Channel}
    Gmap=@n1.Gmap
    Material=@n1.Gmap
   ) {materialFunction=/Game/Weapons/WeaponSkins/MaterialFunction/ClothFuzz}.Result
   False=@n1.Result
  ) {group=Skin FX, parameterName=Use Cloth fuzz}
  n5 = StaticSwitchParameter(
   True=MaterialFunctionCall!(Alpha=ChannelMaskParameter(None=@n3.RGBA) {group=Skin FX, parameterName=Silk Channel}.R, Material=@n4) {materialFunction=/Game/Weapons/WeaponSkins/MaterialFunction/MF_Silk}.Result
   False=@n4
  ) {group=Skin FX, parameterName=Use Silk}
  n6 = MaterialFunctionCall!(Alpha=StaticComponentMaskParameter(None=@n3.RGBA) {defaultR=true, group=Skin FX, parameterName=ThinFilm_Channel}, Material Attributes=@n5) {materialFunction=/Game/Weapons/WeaponSkins/MaterialFunction/ThinFilmFunction}
  n7 = MaterialFunctionCall(
   AO=@n1.AO
   Distance Blend=@n1.AO
   Fresnel=@n1.AO
   Gmap=@n1.AO
   Pre Skinned Local Position=@n1.AO
   Material=MaterialFunctionCall!(
    Alpha=ChannelMaskParameter(None=@n3.RGBA) {parameterName=LensesChannel}.R
    Material=StaticSwitchParameter(True=@n6.Result, False=@n5) {group=Skin FX, parameterName=Use Thin Film}
   ) {materialFunction=&/MF_HumanBeing_Lenses}.Result
  ) {materialFunction=&BaseFN_CharacterFunction_PostFX}
  MaterialFunctionCall(In=@n7.Result) {materialFunction=&CINE_MF_CharPostFX}
 ]
}
