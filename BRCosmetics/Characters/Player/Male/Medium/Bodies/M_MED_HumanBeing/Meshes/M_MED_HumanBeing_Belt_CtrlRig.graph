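# Control Rig for the M_MED_HumanBeing belt that moves the buckle bone, dyn_buckle_MSTR, with the character's hips and legs.
# A parent constraint first blends the buckle between the pelvis and both thighs, favoring the pelvis.
# A sequence then runs a spherical pose reader per thigh, each output weighting a small shift and opposite-signed rotation on the buckle.
# The rest of the imported skeleton, including other dyn_ bones, is left untouched.
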ControlRigBlueprint /BRCosmetics/Characters/Player/Male/Medium/Bodies/M_MED_HumanBeing/Meshes/M_MED_HumanBeing_Belt_CtrlRig
graphs [
 RigVMFunctionLibrary {links [], nodes []}
 RigVMModel {nodes [
   RigUnit_BeginExecution = &RigUnit_BeginExecution()
   n1 = &RigUnit_ParentConstraint(
    AdvancedSettings=(RotationOrderForFilter=XZY)
    Child=(Type=Bone,Name="dyn_buckle_MSTR")
    Parents=((Item=(Type=Bone,Name="pelvis"),Weight=0.4),(Item=(Type=Bone,Name="thigh_l"),Weight=0.3),(Item=(Type=Bone,Name="thigh_r"),Weight=0.3))
    Weight=1
    ExecutePin=@RigUnit_BeginExecution.ExecutePin
   )
   n2 = /Script/ControlRig.RigUnit_SequenceExecution(ExecuteContext=@n1.ExecutePin) {is=struct RigVMUnitNode}
   n3 = &RigUnit_SphericalPoseReader(
    ActiveRegionScaleFactors=(PositiveWidth=3.5,NegativeWidth=0.5,PositiveHeight=3.5,NegativeHeight=4)
    ActiveRegionSize=0.1
    Debug=(bDrawDebug=false,bDraw2D=false,bDrawLocalAxes=false,DebugScale=25,DebugSegments=20,DebugThickness=0.25)
    DriverAxis=(X=-1,Y=0,Z=0)
    DriverItem=(Type=Bone,Name="thigh_l")
    FalloffRegionScaleFactors=(PositiveWidth=1,NegativeWidth=0.9,PositiveHeight=1,NegativeHeight=1)
    FalloffSize=0.45
    FlipHeightScaling=true
    FlipWidthScaling=true
    OptionalParentItem=(Type=Bone,Name="None")
    RotationOffset=(X=90,Y=180,Z=0)
    ExecutePin=@n2.A
   )
   > &RigUnit_ModifyTransforms(
    ItemToModify=((Item=(Type=Bone,Name="dyn_buckle_MSTR"),Transform=(Rotation=(X=0,Y=0,Z=0,W=1),Translation=(X=2,Y=0,Z=0),Scale3D=(X=1,Y=1,Z=1))))
    ItemToModify.0.Transform.Rotation=&RigVMFunction_MathQuaternionFromEuler(Euler=(X=-30,Y=0,Z=15)).Result
    Weight=@n3.OutputParam
   )
   n4 = &RigUnit_SphericalPoseReader(
    ActiveRegionScaleFactors=(PositiveWidth=3.5,NegativeWidth=0.5,PositiveHeight=3.5,NegativeHeight=4)
    ActiveRegionSize=0.1
    Debug=(bDrawDebug=false,DebugScale=25,DebugSegments=20,DebugThickness=0.25)
    DriverAxis=(X=1,Y=0,Z=0)
    DriverItem=(Type=Bone,Name="thigh_r")
    FalloffRegionScaleFactors=(PositiveWidth=1,NegativeWidth=1.1,PositiveHeight=1,NegativeHeight=1)
    FalloffSize=0.45
    FlipHeightScaling=true
    FlipWidthScaling=true
    OptionalParentItem=(Type=Bone,Name="None")
    RotationOffset=(X=-90,Y=0,Z=0)
    ExecutePin=@n2.B
   )
   > &RigUnit_ModifyTransforms(
    ItemToModify=((Item=(Type=Bone,Name="dyn_buckle_MSTR"),Transform=(Rotation=(X=0,Y=0,Z=0,W=1),Translation=(X=2,Y=0,Z=0),Scale3D=(X=1,Y=1,Z=1))))
    ItemToModify.0.Transform.Rotation=&RigVMFunction_MathQuaternionFromEuler(Euler=(X=30,Y=0,Z=-15)).Result
    Weight=@n4.OutputParam
   )
  ]}
]
hierarchy [
 root {loc=<0 0 0>, rot=<0 0 0 1>}
 pelvis {parent=root, loc=<0 -0.7071 93.1191>, rot=<0.0127 0.707 -0.0127 -0.707>}
 spine_01 {parent=pelvis, loc=<6.7494 -0.6388 0>, rot=<0 0 0.0503 -0.9987>}
 spine_02 {parent=spine_01, loc=<7.2493 -0.0308 0>, rot=<0 0 0.0041 -1>}
 spine_03 {parent=spine_02, loc=<6.0084 -0.0167 0>, rot=<0 0 0.0027 -1>}
 spine_04 {parent=spine_03, loc=<8.0999 -0.0224 0>, rot=<0 0 0.0026 -1>}
 spine_05 {parent=spine_04, loc=<10.0119 -0.0184 0>, rot=<0 0 -0.0016 -1>}
 clavicle_l {parent=spine_05, loc=<4.3575 -0.1466 -2.3223>, rot=<0.0476 -0.7289 -0.0102 -0.6829>}
 upperarm_l {parent=clavicle_l, loc=<15.5111 0 0>, rot=<0.0221 -0.3319 -0.0095 -0.943>}
 lowerarm_l {parent=upperarm_l, loc=<26.647 0 0>, rot=<0 0 -0.2447 -0.9696>}
 hand_l {parent=lowerarm_l, loc=<22.344 0 0>, rot=<-0.6794 -0.0546 0.1067 -0.7239>}
 thumb_01_l {parent=hand_l, loc=<1.3118 -0.3191 -2.459>, rot=<0.6546 -0.4042 0.0664 -0.6354>}
 thumb_02_l {parent=thumb_01_l, loc=<5.9336 -0.0448 0.0401>, rot=<-0.0088 0.005 0.2979 -0.9545>}
 thumb_03_l {parent=thumb_02_l, loc=<2.2223 0.3682 0>, rot=<0 0 0.0067 -1>}
 index_metacarpal_l {parent=hand_l, loc=<1.6535 0.0183 -2.96>, rot=<0.0158 -0.0124 -0.0173 -0.9996>}
 index_01_l {parent=index_metacarpal_l, loc=<8.3008 0.0357 -0.0524>, rot=<-0.0044 0.0485 0.0924 -0.9945>}
 index_02_l {parent=index_01_l, loc=<3.3587 -0.3004 0>, rot=<0 0 0.3225 -0.9466>}
 index_03_l {parent=index_02_l, loc=<2.5578 -0.123 0>, rot=<0 0 0.2559 -0.9667>}
 middle_metacarpal_l {parent=hand_l, loc=<2.1856 -0.1725 -0.5984>, rot=<0.006 0.0168 -0.0329 -0.9993>}
 middle_01_l {parent=middle_metacarpal_l, loc=<7.3275 -0.2447 0.0129>, rot=<-0.0208 0.0042 0.0858 -0.9961>}
 middle_02_l {parent=middle_01_l, loc=<4.016 0.1415 0>, rot=<0 0 0.3519 -0.936>}
 middle_03_l {parent=middle_02_l, loc=<2.6103 0.0517 0>, rot=<0 0 0.1707 -0.9853>}
 ring_metacarpal_l {parent=hand_l, loc=<2.2074 -0.3625 1.3638>, rot=<-0.0284 0.0722 -0.0458 -0.9959>}
 ring_01_l {parent=ring_metacarpal_l, loc=<7.0276 -0.6134 -0.2041>, rot=<-0.0484 -0.0225 0.1894 -0.9805>}
 ring_02_l {parent=ring_01_l, loc=<3.6174 0.4647 0>, rot=<0 0 0.2806 -0.9598>}
 ring_03_l {parent=ring_02_l, loc=<2.4152 0.114 0>, rot=<0 0 0.1189 -0.9929>}
 pinky_metacarpal_l {parent=hand_l, loc=<1.7616 -0.9753 3.3814>, rot=<-0.0397 0.0814 -0.069 -0.9935>}
 pinky_01_l {parent=pinky_metacarpal_l, loc=<7.1709 -0.4831 -0.1094>, rot=<-0.0676 -0.0132 0.2236 -0.9723>}
 pinky_02_l {parent=pinky_01_l, loc=<2.8274 0.0893 0>, rot=<0 0 0.2854 -0.9584>}
 pinky_03_l {parent=pinky_02_l, loc=<2.2046 0.0307 0>, rot=<0 0 0.1445 -0.9895>}
 lowerarm_twist_02_l {parent=lowerarm_l, loc=<9.6199 0 0>, rot=<0 0 0 1>}
 lowerarm_twist_01_l {parent=lowerarm_l, loc=<17.8549 0 0>, rot=<0 0 0 1>}
 upperarm_twist_01_l {parent=upperarm_l, loc=<0.0041 0 0>, rot=<0 0 0 1>}
 upperarm_twist_02_l {parent=upperarm_l, loc=<13.7941 0 0>, rot=<0 0 0 1>}
 dyn_sleeve_base_l {parent=upperarm_l, loc=<6.0135 0.5535 0.1481>, rot=<-0.0009 0.0083 -0.0191 -0.9998>}
 dyn_sleeve_l_01_MSTR {parent=dyn_sleeve_base_l, loc=<-0.2674 1.1652 4.1263>, rot=<-0.0074 0.045 0.0016 -0.999>}
 dyn_sleeve_l_01 {parent=dyn_sleeve_l_01_MSTR, loc=<0 0 0>, rot=<0 0 0 1>}
 dyn_sleeve_l_02_MSTR {parent=dyn_sleeve_base_l, loc=<0.6224 4.5546 -0.9048>, rot=<0.5322 0.0497 0.0592 -0.8431>}
 dyn_sleeve_l_02 {parent=dyn_sleeve_l_02_MSTR, loc=<0 0 0>, rot=<0 0 0 1>}
 dyn_sleeve_l_03_MSTR {parent=dyn_sleeve_base_l, loc=<-0.1076 -4.8847 -0.0747>, rot=<-0.5628 0.0674 -0.0607 -0.8216>}
 dyn_sleeve_l_03 {parent=dyn_sleeve_l_03_MSTR, loc=<0 0 0>, rot=<0 0 0 1>}
 dyn_sleeve_l_04_MSTR {parent=dyn_sleeve_base_l, loc=<0.5614 -0.0006 -5.7757>, rot=<0.9983 -0.0003 0.045 -0.0364>}
 dyn_sleeve_l_04 {parent=dyn_sleeve_l_04_MSTR, loc=<0 0 0>, rot=<0 0 0 1>}
 neck_01 {parent=spine_05, loc=<11.3138 0.1529 0>, rot=<0 0 -0.2003 -0.9797>}
 neck_02 {parent=neck_01, loc=<4.5627 0 0>, rot=<0 0 0.0504 -0.9987>}
 head {parent=neck_02, loc=<5.3914 0 0>, rot=<0 0 0.115 -0.9934>}
 clavicle_r {parent=spine_05, loc=<4.3575 -0.1466 2.3223>, rot=<0.7289 0.0476 -0.6829 0.0102>}
 upperarm_r {parent=clavicle_r, loc=<-15.5111 0 0>, rot=<0.0221 -0.3319 -0.0095 -0.943>}
 lowerarm_r {parent=upperarm_r, loc=<-26.647 0 0>, rot=<0 0 -0.2447 -0.9696>}
 hand_r {parent=lowerarm_r, loc=<-22.344 0 0>, rot=<-0.6794 -0.0546 0.1067 -0.7239>}
 thumb_01_r {parent=hand_r, loc=<-1.3118 0.3191 2.459>, rot=<0.6546 -0.4042 0.0664 -0.6354>}
 thumb_02_r {parent=thumb_01_r, loc=<-5.9336 0.0448 -0.0401>, rot=<-0.0088 0.005 0.2979 -0.9545>}
 thumb_03_r {parent=thumb_02_r, loc=<-2.2223 -0.3682 0>, rot=<0 0 0.0067 -1>}
 index_metacarpal_r {parent=hand_r, loc=<-1.6535 -0.0183 2.96>, rot=<0.0158 -0.0124 -0.0173 -0.9996>}
 index_01_r {parent=index_metacarpal_r, loc=<-8.3008 -0.0357 0.0524>, rot=<-0.0044 0.0485 0.0924 -0.9945>}
 index_02_r {parent=index_01_r, loc=<-3.3587 0.3004 0>, rot=<0 0 0.3225 -0.9466>}
 index_03_r {parent=index_02_r, loc=<-2.5578 0.123 0>, rot=<0 0 0.2559 -0.9667>}
 middle_metacarpal_r {parent=hand_r, loc=<-2.1856 0.1725 0.5984>, rot=<0.006 0.0168 -0.0329 -0.9993>}
 middle_01_r {parent=middle_metacarpal_r, loc=<-7.3275 0.2447 -0.0129>, rot=<-0.0208 0.0042 0.0858 -0.9961>}
 middle_02_r {parent=middle_01_r, loc=<-4.016 -0.1415 0>, rot=<0 0 0.3519 -0.936>}
 middle_03_r {parent=middle_02_r, loc=<-2.6103 -0.0517 0>, rot=<0 0 0.1707 -0.9853>}
 ring_metacarpal_r {parent=hand_r, loc=<-2.2074 0.3625 -1.3638>, rot=<-0.0284 0.0722 -0.0458 -0.9959>}
 ring_01_r {parent=ring_metacarpal_r, loc=<-7.0276 0.6134 0.2041>, rot=<-0.0484 -0.0225 0.1894 -0.9805>}
 ring_02_r {parent=ring_01_r, loc=<-3.6174 -0.4647 0>, rot=<0 0 0.2806 -0.9598>}
 ring_03_r {parent=ring_02_r, loc=<-2.4152 -0.114 0>, rot=<0 0 0.1189 -0.9929>}
 pinky_metacarpal_r {parent=hand_r, loc=<-1.7616 0.9753 -3.3814>, rot=<-0.0397 0.0814 -0.069 -0.9935>}
 pinky_01_r {parent=pinky_metacarpal_r, loc=<-7.1709 0.4831 0.1094>, rot=<-0.0676 -0.0132 0.2236 -0.9723>}
 pinky_02_r {parent=pinky_01_r, loc=<-2.8274 -0.0893 0>, rot=<0 0 0.2854 -0.9584>}
 pinky_03_r {parent=pinky_02_r, loc=<-2.2046 -0.0307 0>, rot=<0 0 0.1445 -0.9895>}
 lowerarm_twist_02_r {parent=lowerarm_r, loc=<-9.6199 0 0>, rot=<0 0 0 -1>}
 lowerarm_twist_01_r {parent=lowerarm_r, loc=<-17.8549 0 0>, rot=<0 0 0 -1>}
 upperarm_twist_01_r {parent=upperarm_r, loc=<-0.0041 0 0>, rot=<0 0 0 -1>}
 upperarm_twist_02_r {parent=upperarm_r, loc=<-13.7941 0 0>, rot=<0 0 0 -1>}
 dyn_sleeve_base_r {parent=upperarm_r, loc=<-5.5062 -0.5341 -0.1396>, rot=<0.0191 -0.9998 -0.0009 -0.0083>}
 dyn_sleeve_r_01_MSTR {parent=dyn_sleeve_base_r, loc=<0.3226 -1.1801 3.6454>, rot=<0.0074 0.045 -0.0016 -0.999>}
 dyn_sleeve_r_01 {parent=dyn_sleeve_r_01_MSTR, loc=<0 0 0>, rot=<0 0 0 1>}
 dyn_sleeve_r_02_MSTR {parent=dyn_sleeve_base_r, loc=<1.052 -4.9316 -1.7942>, rot=<-0.5322 0.0497 -0.0592 -0.8431>}
 dyn_sleeve_r_02 {parent=dyn_sleeve_r_02_MSTR, loc=<0 0 0>, rot=<0 0 0 1>}
 dyn_sleeve_r_03_MSTR {parent=dyn_sleeve_base_r, loc=<0.2027 5.0754 -0.2994>, rot=<0.5628 0.0674 0.0607 -0.8216>}
 dyn_sleeve_r_03 {parent=dyn_sleeve_r_03_MSTR, loc=<0 0 0>, rot=<0 0 0 1>}
 dyn_sleeve_r_04_MSTR {parent=dyn_sleeve_base_r, loc=<0.8415 0.0297 -5.494>, rot=<0.9983 0.0003 0.045 0.0364>}
 dyn_sleeve_r_04 {parent=dyn_sleeve_r_04_MSTR, loc=<0 0 0>, rot=<0 0 0 1>}
 dyn_necklace_MSTR {parent=spine_05, loc=<10.2843 6.6835 0.0219>, rot=<-0.3043 -0.9526 0 0>}
 dyn_necklace_1 {parent=dyn_necklace_MSTR, loc=<0 0 0>, rot=<0 0 0 1>}
 dyn_necklace_2 {parent=dyn_necklace_1, loc=<3.8568 0 0>, rot=<0 0 0.0003 -1>}
 dyn_necklace_3 {parent=dyn_necklace_2, loc=<3.8663 0 0>, rot=<0 0 0.2522 -0.9677>}
 dyn_overshirt_MSTR {parent=spine_01, loc=<7.2493 -0.0308 0>, rot=<0 0 0.0041 -1>}
 dyn_overshirt_1_MSTR {parent=dyn_overshirt_MSTR, loc=<0.4339 -13.2158 0>, rot=<0.0365 -0.9993 0 0>}
 dyn_overshirt_1 {parent=dyn_overshirt_1_MSTR, loc=<0 0 0>, rot=<0 0 0 1>}
 dyn_overshirt_2_MSTR {parent=dyn_overshirt_MSTR, loc=<1.704 -10.4823 10.0223>, rot=<0.0347 -0.9504 0.3088 -0.0113>}
 dyn_overshirt_2 {parent=dyn_overshirt_2_MSTR, loc=<0 0 0>, rot=<0 0 0 1>}
 dyn_overshirt_3_MSTR {parent=dyn_overshirt_MSTR, loc=<2.3307 -0.1299 14.5114>, rot=<0.0258 -0.7066 0.7066 -0.0258>}
 dyn_overshirt_3 {parent=dyn_overshirt_3_MSTR, loc=<0 0 0>, rot=<0 0 0 1>}
 dyn_overshirt_4_MSTR {parent=dyn_overshirt_MSTR, loc=<2.4515 8.0784 9.2649>, rot=<0.0095 -0.2586 0.9653 -0.0353>}
 dyn_overshirt_4 {parent=dyn_overshirt_4_MSTR, loc=<0 0 0>, rot=<0 0 0 1>}
 dyn_overshirt_5_MSTR {parent=dyn_overshirt_MSTR, loc=<-0.5726 9.3836 0>, rot=<0 0 0.9993 -0.0365>}
 dyn_overshirt_5 {parent=dyn_overshirt_5_MSTR, loc=<0 0 0>, rot=<0 0 0 1>}
 dyn_overshirt_6_MSTR {parent=dyn_overshirt_MSTR, loc=<-1.687 -10.8512 -10.0984>, rot=<0.0347 -0.9504 -0.3088 0.0113>}
 dyn_overshirt_6 {parent=dyn_overshirt_6_MSTR, loc=<0 0 0>, rot=<0 0 0 -1>}
 dyn_overshirt_7_MSTR {parent=dyn_overshirt_MSTR, loc=<-3.0097 -0.5207 -14.9318>, rot=<-0.0258 0.7066 0.7066 -0.0258>}
 dyn_overshirt_7 {parent=dyn_overshirt_7_MSTR, loc=<0 0 0>, rot=<0 0 0 1>}
 dyn_overshirt_8_MSTR {parent=dyn_overshirt_MSTR, loc=<-2.4301 7.8261 -9.3432>, rot=<-0.0095 0.2586 0.9653 -0.0353>}
 dyn_overshirt_8 {parent=dyn_overshirt_8_MSTR, loc=<0 0 0>, rot=<0 0 0 1>}
 thigh_l {parent=pelvis, loc=<-2.5693 -0.0109 -7.7501>, rot=<0.0013 0.0462 0.0155 -0.9988>}
 calf_l {parent=thigh_l, loc=<-39.039 0 0>, rot=<0 0 -0.0378 -0.9993>}
 foot_l {parent=calf_l, loc=<-42.5546 0 0>, rot=<0.0427 -0.0505 0.0328 -0.9973>}
 ball_l {parent=foot_l, loc=<-6.1599 13.8264 0.6806>, rot=<0.0001 -0.0001 -0.7063 -0.7079>}
 calf_twist_02_l {parent=calf_l, loc=<-18.8064 -0.0967 0.0397>, rot=<0.0237 0.0095 -0.0035 -0.9997>}
 calf_twist_01_l {parent=calf_l, loc=<-35.813 -0.0967 0.0397>, rot=<0.0237 0.0095 -0.0035 -0.9997>}
 thigh_twist_01_l {parent=thigh_l, loc=<-21.629 -0.0919 -0.0002>, rot=<0.0013 0 0 -1>}
 thigh_r {parent=pelvis, loc=<-2.5693 -0.0109 7.7501>, rot=<0.0462 -0.0013 0.9988 0.0155>}
 calf_r {parent=thigh_r, loc=<39.0389 0 0>, rot=<0 0 -0.0378 -0.9993>}
 foot_r {parent=calf_r, loc=<42.5546 0 0>, rot=<0.0427 -0.0505 0.0328 -0.9973>}
 ball_r {parent=foot_r, loc=<6.1599 -13.8264 -0.6806>, rot=<0.0001 -0.0001 -0.7063 -0.7079>}
 calf_twist_02_r {parent=calf_r, loc=<18.8064 0.0967 -0.0397>, rot=<0.0237 0.0095 -0.0035 -0.9997>}
 calf_twist_01_r {parent=calf_r, loc=<35.813 0.0967 -0.0397>, rot=<0.0237 0.0095 -0.0035 -0.9997>}
 thigh_twist_01_r {parent=thigh_r, loc=<21.629 0.0919 0.0002>, rot=<0.0013 0 0 -1>}
 dyn_strap_MSTR {parent=thigh_r, loc=<18.6092 -2.8549 9.033>, rot=<0.8155 0.4627 0.3018 0.1728>}
 dyn_strap_1 {parent=dyn_strap_MSTR, loc=<0 0 0>, rot=<0 0 0 1>}
 dyn_strap_2 {parent=dyn_strap_1, loc=<6.0588 0 0>, rot=<-0.0633 0.1941 -0.6469 -0.7348>}
 dyn_strap_3 {parent=dyn_strap_2, loc=<7.5482 0 0>, rot=<-0.0236 0.1559 0.522 -0.8382>}
 dyn_strap_4 {parent=dyn_strap_3, loc=<7.3524 0 0>, rot=<0.0038 -0.0559 -0.5964 -0.8008>}
 dyn_undershirt_MSTR {parent=pelvis, loc=<0 0 0>, rot=<0 0 0 1>}
 dyn_undershirt_1_MSTR {parent=dyn_undershirt_MSTR, loc=<-1.3433 -12.7243 0>, rot=<-0.0179 -0.9998 0 0>}
 dyn_undershirt_1 {parent=dyn_undershirt_1_MSTR, loc=<0 0 0>, rot=<0 0 0 1>}
 dyn_undershirt_2_MSTR {parent=dyn_undershirt_MSTR, loc=<-1.9519 -9.5568 10.4412>, rot=<-0.017 -0.9509 0.309 0.0055>}
 dyn_undershirt_2 {parent=dyn_undershirt_2_MSTR, loc=<0 0 0>, rot=<0 0 0 1>}
 dyn_undershirt_3_MSTR {parent=dyn_undershirt_MSTR, loc=<-2.6445 0.1613 14.5608>, rot=<-0.0127 -0.707 0.707 0.0127>}
 dyn_undershirt_3 {parent=dyn_undershirt_3_MSTR, loc=<0 0 0>, rot=<0 0 0 1>}
 dyn_undershirt_4_MSTR {parent=dyn_undershirt_MSTR, loc=<-0.3714 8.2413 10.1088>, rot=<-0.0046 -0.2588 0.9658 0.0173>}
 dyn_undershirt_4 {parent=dyn_undershirt_4_MSTR, loc=<0 0 0>, rot=<0 0 0 1>}
 dyn_undershirt_5_MSTR {parent=dyn_undershirt_MSTR, loc=<-0.8307 8.8279 0>, rot=<0 0 0.9998 0.0179>}
 dyn_undershirt_5 {parent=dyn_undershirt_5_MSTR, loc=<0 0 0>, rot=<0 0 0 1>}
 dyn_undershirt_6_MSTR {parent=dyn_undershirt_MSTR, loc=<-0.2969 -10.0157 -10.7313>, rot=<-0.017 -0.9509 -0.309 -0.0055>}
 dyn_undershirt_6 {parent=dyn_undershirt_6_MSTR, loc=<0 0 0>, rot=<0 0 0 -1>}
 dyn_undershirt_7_MSTR {parent=dyn_undershirt_MSTR, loc=<-0.7149 0.0922 -14.1398>, rot=<0.0127 0.707 0.707 0.0127>}
 dyn_undershirt_7 {parent=dyn_undershirt_7_MSTR, loc=<0 0 0>, rot=<0 0 0 1>}
 dyn_undershirt_8_MSTR {parent=dyn_undershirt_MSTR, loc=<-0.8687 6.3866 -9.0284>, rot=<0.0046 0.2588 0.9658 0.0173>}
 dyn_undershirt_8 {parent=dyn_undershirt_8_MSTR, loc=<0 0 0>, rot=<0 0 0 1>}
 dyn_buckle_MSTR {parent=pelvis, loc=<-9.7302 11.1136 1.7344>, rot=<-0.5089 -0.491 -0.491 -0.5089>}
 dyn_buckle_01 {parent=dyn_buckle_MSTR, loc=<0 0 0>, rot=<0 0 0 1>}
]
