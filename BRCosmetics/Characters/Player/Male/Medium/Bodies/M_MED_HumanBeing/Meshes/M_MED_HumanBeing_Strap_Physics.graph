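# Physics asset for the M_MED_HumanBeing male medium cosmetic body, defining bodies for a strap and belt on the right thigh, apparently for secondary motion.
# A four-capsule strap chain and a one-capsule belt each start at a small MSTR sphere constrained to thigh_r with free rotation.
# Narrow swing and twist limits keep both chains fairly stiff.
# Large pelvis and thigh capsules, also joined with free rotation, supply the body shapes.
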PhysicsAsset /BRCosmetics/Characters/Player/Male/Medium/Bodies/M_MED_HumanBeing/Meshes/M_MED_HumanBeing_Strap_Physics
bodies {
 dyn_belt_1 [{shapeType=Capsule, length=3, radius=2, shapeName=dyn_belt_1_capsule, center=<1 0 0>, rotation=<rot 90 0 0>}]
 dyn_belt_MSTR [{shapeType=Sphere, radius=0.505, shapeName=dyn_belt_MSTR_sphere}]
 dyn_strap_1 [{shapeType=Capsule, length=4, radius=2, shapeName=dyn_strap_1_capsule, center=<2 0 0>, rotation=<rot 90 0 0>}]
 dyn_strap_2 [{shapeType=Capsule, length=4, radius=2, shapeName=dyn_strap_2_capsule, center=<2 0 0>, rotation=<rot 90 0 0>}]
 dyn_strap_3 [{shapeType=Capsule, length=4, radius=2, shapeName=dyn_strap_3_capsule, center=<2 0 0>, rotation=<rot 90 0 0>}]
 dyn_strap_4 [{shapeType=Capsule, length=4, radius=2, shapeName=dyn_strap_4_capsule, center=<2 0 0>, rotation=<rot 90 0 0>}]
 dyn_strap_MSTR [{shapeType=Sphere, radius=0.505, shapeName=dyn_strap_MSTR_sphere}]
 pelvis [{shapeType=Capsule, length=10, radius=12, shapeName=pelvis_capsule, center=<-10.1563 0 0>}]
 thigh_r [{shapeType=Capsule, length=25, radius=10, shapeName=thigh_r_capsule, center=<27.9749 0 0>, rotation=<rot 90 0 0>}]
}
constraints [
 {bone1Name=dyn_strap_1, bone2Name=dyn_strap_MSTR, swing1LimitDegrees=10, swing2LimitDegrees=10, twistLimitDegrees=1}
 {bone1Name=dyn_strap_2, bone2Name=dyn_strap_1, swing1LimitDegrees=10, swing2LimitDegrees=10, twistLimitDegrees=1}
 {bone1Name=dyn_strap_3, bone2Name=dyn_strap_2, swing1LimitDegrees=10, swing2LimitDegrees=10, twistLimitDegrees=1}
 {bone1Name=dyn_strap_4, bone2Name=dyn_strap_3, swing1LimitDegrees=10, swing2LimitDegrees=10, twistLimitDegrees=1}
 {
  bone1Name=dyn_strap_MSTR
  bone2Name=thigh_r
  swing1LimitDegrees=45
  swing1Motion=Free
  swing2LimitDegrees=45
  swing2Motion=Free
  twistLimitDegrees=45
  twistMotion=Free
 }
 {
  bone1Name=thigh_r
  bone2Name=pelvis
  swing1LimitDegrees=45
  swing1Motion=Free
  swing2LimitDegrees=45
  swing2Motion=Free
  twistLimitDegrees=45
  twistMotion=Free
 }
 {bone1Name=dyn_belt_1, bone2Name=dyn_belt_MSTR, swing1LimitDegrees=20, swing2LimitDegrees=1, twistLimitDegrees=1}
 {
  bone1Name=dyn_belt_MSTR
  bone2Name=thigh_r
  swing1LimitDegrees=45
  swing1Motion=Free
  swing2LimitDegrees=45
  swing2Motion=Free
  twistLimitDegrees=45
  twistMotion=Free
 }
]
