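# Physics asset for the upper mesh of the medium male HumanBeing cosmetic body, covering necklace and sleeve dynamic bones.
# Capsules on spine_04 and spine_05 form the torso, with sphere roots for necklace and both sleeves linked to spine_05 by free angular constraints.
# The necklace is a three-capsule chain with moderate swing limits.
# Each sleeve has four parallel capsules constrained directly to its base with tight swing and twist limits.
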PhysicsAsset /BRCosmetics/Characters/Player/Male/Medium/Bodies/M_MED_HumanBeing/Meshes/M_MED_HumanBeing_Upper_Physics
bodies {
 dyn_necklace_1 [{shapeType=Capsule, length=8, radius=8, shapeName=dyn_necklace_1_capsule, center=<4 7 0>}]
 dyn_necklace_2 [{shapeType=Capsule, length=5, radius=8, shapeName=dyn_necklace_2_capsule, center=<4 7 0>}]
 dyn_necklace_3 [{shapeType=Capsule, length=5, radius=8, shapeName=dyn_necklace_2_capsule, center=<6 7 0>}]
 dyn_necklace_MSTR [{shapeType=Sphere, radius=0.505, shapeName=dyn_necklace_MSTR_sphere}]
 dyn_sleeve_base_l [{shapeType=Sphere, radius=0.505, shapeName=dyn_sleeve_base_l_sphere}]
 dyn_sleeve_base_r [{shapeType=Sphere, radius=0.505, shapeName=dyn_sleeve_base_r_sphere}]
 dyn_sleeve_l_01 [{shapeType=Capsule, length=5, radius=4, shapeName=dyn_sleeve_l_01_capsule, center=<3 0 -2>, rotation=<rot 90 0 0>}]
 dyn_sleeve_l_02 [{shapeType=Capsule, length=5, radius=4, shapeName=dyn_sleeve_l_02_capsule, center=<3 0 -2>, rotation=<rot 90 0 0>}]
 dyn_sleeve_l_03 [{shapeType=Capsule, length=5, radius=4, shapeName=dyn_sleeve_l_03_capsule, center=<3 0 -2>, rotation=<rot 90 0 0>}]
 dyn_sleeve_l_04 [{shapeType=Capsule, length=5, radius=4, shapeName=dyn_sleeve_l_04_capsule, center=<3 0 -2>, rotation=<rot 90 0 0>}]
 dyn_sleeve_r_01 [{shapeType=Capsule, length=5, radius=4, shapeName=dyn_sleeve_r_01_capsule, center=<3 0 -2>, rotation=<rot 90 0 0>}]
 dyn_sleeve_r_02 [{shapeType=Capsule, length=5, radius=4, shapeName=dyn_sleeve_r_02_capsule, center=<3 0 -2>, rotation=<rot 90 0 0>}]
 dyn_sleeve_r_03 [{shapeType=Capsule, length=5, radius=4, shapeName=dyn_sleeve_r_03_capsule, center=<3 0 -2>, rotation=<rot 90 0 0>}]
 dyn_sleeve_r_04 [{shapeType=Capsule, length=5, radius=4, shapeName=dyn_sleeve_r_04_capsule, center=<3 0 -2>, rotation=<rot 90 0 0>}]
 spine_04 [{shapeType=Capsule, length=6, radius=12, shapeName=spine_04_capsule, center=<0 -2 0>}]
 spine_05 [{shapeType=Capsule, length=8, radius=12, shapeName=spine_05_capsule, center=<0 -2.5 0>}]
}
constraints [
 {bone1Name=dyn_sleeve_r_01, bone2Name=dyn_sleeve_base_r, swing1LimitDegrees=3, swing2LimitDegrees=8, twistLimitDegrees=1}
 {bone1Name=dyn_sleeve_r_02, bone2Name=dyn_sleeve_base_r, swing1LimitDegrees=3, swing2LimitDegrees=8, twistLimitDegrees=1}
 {bone1Name=dyn_sleeve_r_03, bone2Name=dyn_sleeve_base_r, swing1LimitDegrees=3, swing2LimitDegrees=8, twistLimitDegrees=1}
 {bone1Name=dyn_sleeve_r_04, bone2Name=dyn_sleeve_base_r, swing1LimitDegrees=3, swing2LimitDegrees=8, twistLimitDegrees=1}
 {bone1Name=dyn_sleeve_l_01, bone2Name=dyn_sleeve_base_l, swing1LimitDegrees=3, swing2LimitDegrees=8, twistLimitDegrees=1}
 {bone1Name=dyn_sleeve_l_03, bone2Name=dyn_sleeve_base_l, swing1LimitDegrees=3, swing2LimitDegrees=8, twistLimitDegrees=1}
 {bone1Name=dyn_sleeve_l_02, bone2Name=dyn_sleeve_base_l, swing1LimitDegrees=3, swing2LimitDegrees=8, twistLimitDegrees=1}
 {bone1Name=dyn_sleeve_l_04, bone2Name=dyn_sleeve_base_l, swing1LimitDegrees=3, swing2LimitDegrees=8, twistLimitDegrees=1}
 {bone1Name=dyn_necklace_2, bone2Name=dyn_necklace_1, swing1LimitDegrees=30, swing2LimitDegrees=15, twistLimitDegrees=1}
 {bone1Name=dyn_necklace_1, bone2Name=dyn_necklace_MSTR, swing1LimitDegrees=15, swing2LimitDegrees=15, twistLimitDegrees=1}
 {
  bone1Name=dyn_sleeve_base_l
  bone2Name=spine_05
  swing1LimitDegrees=45
  swing1Motion=Free
  swing2LimitDegrees=45
  swing2Motion=Free
  twistLimitDegrees=45
  twistMotion=Free
 }
 {
  bone1Name=dyn_sleeve_base_r
  bone2Name=spine_05
  swing1LimitDegrees=45
  swing1Motion=Free
  swing2LimitDegrees=45
  swing2Motion=Free
  twistLimitDegrees=45
  twistMotion=Free
 }
 {
  bone1Name=dyn_necklace_MSTR
  bone2Name=spine_05
  swing1LimitDegrees=45
  swing1Motion=Free
  swing2LimitDegrees=45
  swing2Motion=Free
  twistLimitDegrees=45
  twistMotion=Free
 }
 {
  bone1Name=spine_05
  bone2Name=spine_04
  swing1LimitDegrees=45
  swing1Motion=Free
  swing2LimitDegrees=45
  swing2Motion=Free
  twistLimitDegrees=45
  twistMotion=Free
 }
 {bone1Name=dyn_necklace_3, bone2Name=dyn_necklace_2, swing1LimitDegrees=30, swing2LimitDegrees=15, twistLimitDegrees=1}
]
