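# Animation Blueprint for the M_MED_HumanBeing player body cosmetic, adding secondary motion to belt, shirts and straps.
# The parent mesh pose feeds a belt Control Rig, chained AnimDynamics on undershirt and overshirt bones, then strap and upper-body RigidBody simulations.
# The event graph drives material world-position-offset cloth from foot and calf transforms, with wind while skydive floating.
# It also sets cape, look-down, menu-idle and pant-leg values, which the visible graphs never read.
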AnimBlueprint /BRCosmetics/Characters/Player/Male/Medium/Bodies/M_MED_HumanBeing/Meshes/M_MED_HumanBeing_AnimBP
target_skeleton=&/M_MED_HumanBeing_Skeleton
graphs [
 AnimGraph {nodes [
   n1 = Control Rig(Source=Copy Pose From Mesh() {bUseAttachedParent=true}) {
    alphaBoolBlend {blendOption=Linear}
    controlRigAssetReference {blueprintRigClass=&/M_MED_HumanBeing_Belt_CtrlRig.M_MED_HumanBeing_Belt_CtrlRig_C}
   }
   n2 = Anim Dynamics - Bone: dyn_undershirt_1(ComponentPose=Local To Component(LocalPose=@n1), Alpha=1, ExternalForce="0,0,0") {
    bUseGravityOverride=true
    alphaBoolBlend {blendOption=Linear}
    boundBone {boneName=dyn_undershirt_1}
    componentAppliedLinearAccClamp=<3000 3000 3000>
    componentLinearAccScale=<0.1 0.1 0.2>
    componentLinearVelScale=<0.1 0.1 0.2>
    gravityOverride {z=-980}
    physicsBodyDefinitions [{
      boundBone {boneName=dyn_undershirt_1}
      boxExtents=<20 5 10>
      constraintSetup {angularLimitsMax=<0 5 5>, angularLimitsMin=<0 -5 -20>, angularTarget=<1 0 0>}
      localJointOffset=<4 -0.1 0>
     }]
   }
   n3 = Anim Dynamics - Bone: dyn_undershirt_2(ComponentPose=@n2, Alpha=1, ExternalForce="0,0,0") {
    bUseGravityOverride=true
    alphaBoolBlend {blendOption=Linear}
    boundBone {boneName=dyn_undershirt_2}
    componentAppliedLinearAccClamp=<3000 3000 3000>
    componentLinearAccScale=<0.1 0.1 0.2>
    componentLinearVelScale=<0.1 0.1 0.2>
    gravityOverride {z=-980}
    physicsBodyDefinitions [{
      boundBone {boneName=dyn_undershirt_2}
      boxExtents=<20 5 10>
      constraintSetup {angularLimitsMax=<0 5 5>, angularLimitsMin=<0 -5 -20>, angularTarget=<1 0 0>}
      localJointOffset=<4 -0.1 0>
     }]
   }
   n4 = Anim Dynamics - Bone: dyn_undershirt_3(ComponentPose=@n3, Alpha=1, ExternalForce="0,0,0") {
    bUseGravityOverride=true
    alphaBoolBlend {blendOption=Linear}
    boundBone {boneName=dyn_undershirt_3}
    componentAppliedLinearAccClamp=<3000 3000 3000>
    componentLinearAccScale=<0.1 0.1 0.2>
    componentLinearVelScale=<0.1 0.1 0.2>
    gravityOverride {z=-980}
    physicsBodyDefinitions [{
      boundBone {boneName=dyn_undershirt_3}
      boxExtents=<20 5 10>
      constraintSetup {angularLimitsMax=<0 5 5>, angularLimitsMin=<0 -5 -20>, angularTarget=<1 0 0>}
      localJointOffset=<4 -0.1 0>
     }]
   }
   n5 = Anim Dynamics - Bone: dyn_undershirt_4(ComponentPose=@n4, Alpha=1, ExternalForce="0,0,0") {
    bUseGravityOverride=true
    alphaBoolBlend {blendOption=Linear}
    boundBone {boneName=dyn_undershirt_4}
    componentAppliedLinearAccClamp=<3000 3000 3000>
    componentLinearAccScale=<0.1 0.1 0.2>
    componentLinearVelScale=<0.1 0.1 0.2>
    gravityOverride {z=-980}
    physicsBodyDefinitions [{
      boundBone {boneName=dyn_undershirt_4}
      boxExtents=<20 5 10>
      constraintSetup {angularLimitsMax=<0 5 5>, angularLimitsMin=<0 -5 -20>, angularTarget=<1 0 0>}
      localJointOffset=<4 -0.1 0>
     }]
   }
   n6 = Anim Dynamics - Bone: dyn_undershirt_5(ComponentPose=@n5, Alpha=1, ExternalForce="0,0,0") {
    bUseGravityOverride=true
    alphaBoolBlend {blendOption=Linear}
    boundBone {boneName=dyn_undershirt_5}
    componentAppliedLinearAccClamp=<3000 3000 3000>
    componentLinearAccScale=<0.1 0.1 0.2>
    componentLinearVelScale=<0.1 0.1 0.2>
    gravityOverride {z=-980}
    physicsBodyDefinitions [{
      boundBone {boneName=dyn_undershirt_5}
      boxExtents=<20 5 10>
      constraintSetup {angularLimitsMax=<0 5 5>, angularLimitsMin=<0 -5 -20>, angularTarget=<1 0 0>}
      localJointOffset=<4 -0.1 0>
     }]
   }
   n7 = Anim Dynamics - Bone: dyn_undershirt_6(ComponentPose=@n6, Alpha=1, ExternalForce="0,0,0") {
    bUseGravityOverride=true
    alphaBoolBlend {blendOption=Linear}
    boundBone {boneName=dyn_undershirt_6}
    componentAppliedLinearAccClamp=<3000 3000 3000>
    componentLinearAccScale=<0.1 0.1 0.2>
    componentLinearVelScale=<0.1 0.1 0.2>
    gravityOverride {z=-980}
    physicsBodyDefinitions [{
      boundBone {boneName=dyn_undershirt_6}
      boxExtents=<20 5 10>
      constraintSetup {angularLimitsMax=<0 5 5>, angularLimitsMin=<0 -5 -20>, angularTarget=<1 0 0>}
      localJointOffset=<4 -0.1 0>
     }]
   }
   n8 = Anim Dynamics - Bone: dyn_undershirt_7(ComponentPose=@n7, Alpha=1, ExternalForce="0,0,0") {
    bUseGravityOverride=true
    alphaBoolBlend {blendOption=Linear}
    boundBone {boneName=dyn_undershirt_7}
    componentAppliedLinearAccClamp=<3000 3000 3000>
    componentLinearAccScale=<0.1 0.1 0.2>
    componentLinearVelScale=<0.1 0.1 0.2>
    gravityOverride {z=-980}
    physicsBodyDefinitions [{
      boundBone {boneName=dyn_undershirt_7}
      boxExtents=<20 5 10>
      constraintSetup {angularLimitsMax=<0 5 5>, angularLimitsMin=<0 -5 -20>, angularTarget=<1 0 0>}
      localJointOffset=<4 -0.1 0>
     }]
   }
   n9 = Anim Dynamics - Bone: dyn_undershirt_8(ComponentPose=@n8, Alpha=1, ExternalForce="0,0,0") {
    bUseGravityOverride=true
    alphaBoolBlend {blendOption=Linear}
    boundBone {boneName=dyn_undershirt_8}
    componentAppliedLinearAccClamp=<3000 3000 3000>
    componentLinearAccScale=<0.1 0.1 0.2>
    componentLinearVelScale=<0.1 0.1 0.2>
    gravityOverride {z=-980}
    physicsBodyDefinitions [{
      boundBone {boneName=dyn_undershirt_8}
      boxExtents=<20 5 10>
      constraintSetup {angularLimitsMax=<0 5 5>, angularLimitsMin=<0 -5 -20>, angularTarget=<1 0 0>}
      localJointOffset=<4 -0.1 0>
     }]
   }
   n10 = Anim Dynamics - Bone: dyn_overshirt_1(ComponentPose=@n9, Alpha=1, ExternalForce="0,0,0") {
    bUseGravityOverride=true
    alphaBoolBlend {blendOption=Linear}
    boundBone {boneName=dyn_overshirt_1}
    componentAppliedLinearAccClamp=<3000 3000 3000>
    componentLinearAccScale=<0.1 0.1 0.2>
    componentLinearVelScale=<0.1 0.1 0.2>
    gravityOverride {z=-980}
    physicsBodyDefinitions [{
      boundBone {boneName=dyn_overshirt_1}
      boxExtents=<20 5 10>
      constraintSetup {angularLimitsMax=<0 5 5>, angularLimitsMin=<0 -5 -20>, angularTarget=<1 0 0>}
      localJointOffset=<4 -0.1 0>
     }]
   }
   n11 = Anim Dynamics - Bone: dyn_overshirt_2(ComponentPose=@n10, Alpha=1, ExternalForce="0,0,0") {
    bUseGravityOverride=true
    alphaBoolBlend {blendOption=Linear}
    boundBone {boneName=dyn_overshirt_2}
    componentAppliedLinearAccClamp=<3000 3000 3000>
    componentLinearAccScale=<0.1 0.1 0.2>
    componentLinearVelScale=<0.1 0.1 0.2>
    gravityOverride {z=-980}
    physicsBodyDefinitions [{
      boundBone {boneName=dyn_overshirt_2}
      boxExtents=<20 5 10>
      constraintSetup {angularLimitsMax=<0 5 5>, angularLimitsMin=<0 -5 -20>, angularTarget=<1 0 0>}
      localJointOffset=<4 -0.1 0>
     }]
   }
   n12 = Anim Dynamics - Bone: dyn_overshirt_3(ComponentPose=@n11, Alpha=1, ExternalForce="0,0,0") {
    bUseGravityOverride=true
    alphaBoolBlend {blendOption=Linear}
    boundBone {boneName=dyn_overshirt_3}
    componentAppliedLinearAccClamp=<3000 3000 3000>
    componentLinearAccScale=<0.1 0.1 0.2>
    componentLinearVelScale=<0.1 0.1 0.2>
    gravityOverride {z=-980}
    physicsBodyDefinitions [{
      boundBone {boneName=dyn_overshirt_3}
      boxExtents=<20 5 10>
      constraintSetup {angularLimitsMax=<0 5 5>, angularLimitsMin=<0 -5 -20>, angularTarget=<1 0 0>}
      localJointOffset=<4 -0.1 0>
     }]
   }
   n13 = Anim Dynamics - Bone: dyn_overshirt_4(ComponentPose=@n12, Alpha=1, ExternalForce="0,0,0") {
    bUseGravityOverride=true
    alphaBoolBlend {blendOption=Linear}
    boundBone {boneName=dyn_overshirt_4}
    componentAppliedLinearAccClamp=<3000 3000 3000>
    componentLinearAccScale=<0.1 0.1 0.2>
    componentLinearVelScale=<0.1 0.1 0.2>
    gravityOverride {z=-980}
    physicsBodyDefinitions [{
      boundBone {boneName=dyn_overshirt_4}
      boxExtents=<20 5 10>
      constraintSetup {angularLimitsMax=<0 5 5>, angularLimitsMin=<0 -5 -20>, angularTarget=<1 0 0>}
      localJointOffset=<4 -0.1 0>
     }]
   }
   n14 = Anim Dynamics - Bone: dyn_overshirt_5(ComponentPose=@n13, Alpha=1, ExternalForce="0,0,0") {
    bUseGravityOverride=true
    alphaBoolBlend {blendOption=Linear}
    boundBone {boneName=dyn_overshirt_5}
    componentAppliedLinearAccClamp=<3000 3000 3000>
    componentLinearAccScale=<0.1 0.1 0.2>
    componentLinearVelScale=<0.1 0.1 0.2>
    gravityOverride {z=-980}
    physicsBodyDefinitions [{
      boundBone {boneName=dyn_overshirt_5}
      boxExtents=<20 5 10>
      constraintSetup {angularLimitsMax=<0 5 5>, angularLimitsMin=<0 -5 -20>, angularTarget=<1 0 0>}
      localJointOffset=<4 -0.1 0>
     }]
   }
   n15 = Anim Dynamics - Bone: dyn_overshirt_6(ComponentPose=@n14, Alpha=1, ExternalForce="0,0,0") {
    bUseGravityOverride=true
    alphaBoolBlend {blendOption=Linear}
    boundBone {boneName=dyn_overshirt_6}
    componentAppliedLinearAccClamp=<3000 3000 3000>
    componentLinearAccScale=<0.1 0.1 0.2>
    componentLinearVelScale=<0.1 0.1 0.2>
    gravityOverride {z=-980}
    physicsBodyDefinitions [{
      boundBone {boneName=dyn_overshirt_6}
      boxExtents=<20 5 10>
      constraintSetup {angularLimitsMax=<0 5 5>, angularLimitsMin=<0 -5 -20>, angularTarget=<1 0 0>}
      localJointOffset=<4 -0.1 0>
     }]
   }
   n16 = Anim Dynamics - Bone: dyn_overshirt_7(ComponentPose=@n15, Alpha=1, ExternalForce="0,0,0") {
    bUseGravityOverride=true
    alphaBoolBlend {blendOption=Linear}
    boundBone {boneName=dyn_overshirt_7}
    componentAppliedLinearAccClamp=<3000 3000 3000>
    componentLinearAccScale=<0.1 0.1 0.2>
    componentLinearVelScale=<0.1 0.1 0.2>
    gravityOverride {z=-980}
    physicsBodyDefinitions [{
      boundBone {boneName=dyn_overshirt_7}
      boxExtents=<20 5 10>
      constraintSetup {angularLimitsMax=<0 5 5>, angularLimitsMin=<0 -5 -20>, angularTarget=<1 0 0>}
      localJointOffset=<4 -0.1 0>
     }]
   }
   n17 = Anim Dynamics - Bone: dyn_overshirt_8(ComponentPose=@n16, Alpha=1, ExternalForce="0,0,0") {
    bUseGravityOverride=true
    alphaBoolBlend {blendOption=Linear}
    boundBone {boneName=dyn_overshirt_8}
    componentAppliedLinearAccClamp=<3000 3000 3000>
    componentLinearAccScale=<0.1 0.1 0.2>
    componentLinearVelScale=<0.1 0.1 0.2>
    gravityOverride {z=-980}
    physicsBodyDefinitions [{
      boundBone {boneName=dyn_overshirt_8}
      boxExtents=<20 5 10>
      constraintSetup {angularLimitsMax=<0 5 5>, angularLimitsMin=<0 -5 -20>, angularTarget=<1 0 0>}
      localJointOffset=<4 -0.1 0>
     }]
   }
   n18 = RigidBody(ComponentPose=@n17) {
    bOverrideWorldGravity=true
    overridePhysicsAsset=&/M_MED_HumanBeing_Strap_Physics
    simulationSpace=BaseBoneSpace
    alphaBoolBlend {blendOption=Linear}
    baseBoneRef {boneName=pelvis}
    componentAppliedLinearAccClamp=<3000 3000 3000>
    componentLinearAccScale=<0.2 0.2 0.4>
    componentLinearVelScale=<0.1 0.1 0.2>
    overrideWorldGravity {z=-980}
   }
   n19 = RigidBody(ComponentPose=@n18) {
    bOverrideWorldGravity=true
    overridePhysicsAsset=&/M_MED_HumanBeing_Upper_Physics
    simulationSpace=BaseBoneSpace
    alphaBoolBlend {blendOption=Linear}
    baseBoneRef {boneName=spine_05}
    componentAppliedLinearAccClamp=<3000 3000 3000>
    componentLinearAccScale=<0.2 0.2 0.4>
    componentLinearVelScale=<0.1 0.1 0.2>
    overrideWorldGravity {z=-980}
   }
   Output Pose(Result=Component To Local(ComponentPose=@n19)) {layerGroup=None}
  ]}
 EventGraph {nodes [
   Event BlueprintInitializeAnimation()
   > Set SkelMesh(SkelMesh=GetOwnerSkeletalMeshComponent())
   n1 = TryGetPawnOwner()
   > Is Valid(InputObject=@n1)
   > n2 = Cast To FortPlayerPawnAthena(Object=@n1)
   > Set PlayerPawn(PlayerPawn:Fort Player Pawn Athena Object Reference=@n2) -> (Output_Get:Fort Player Pawn Athena Object Reference)
   Event OnCharacterCustomizationLoaded()
   > n3 = Sequence()
   > n4 = Set SkelMesh(SkelMesh=GetPartSkeletalMeshComponent())
   > n5 = Get Dynamic Material Instance(SkeletalMeshComponent=@n4)
   > Set Mat Instance Dynamic(Mat Instance Dynamic=@n5)
   > n6 = GetOrCreateDynamicOverlayMaterial(MeshComponent=@n4, MaterialName=None)
   > Set Overlay Mat Instance Dynamic(Overlay Mat Instance Dynamic=@n6)
   n7 = Branch(execute=@n3.then_1, Condition=CheckForBackbling().bIsBackblingACape)
   > n8 = Branch(Condition=BackpackItemDefinitionHasAnyMetaTag!(Tags:Array of Gameplay Tag Structures=Make Array("[0]"=(TagName="Cosmetics.MetaTags.ArcticIce.Cape")), bRequiresEquipped:bool=false) -> (ReturnValue:bool))
   > Set custom_cape_equipped(custom_cape_equipped:bool=true) -> (Output_Get:bool)
   Set custom_cape_equipped(execute=@n8.else, custom_cape_equipped=false)
   Set custom_cape_equipped(execute=@n7.else, custom_cape_equipped=false)
   n9 = Event BlueprintUpdateAnimation()
   > n10 = Sequence!(execute) -> (then_0, then_1, then_2:Exec, then_3:Exec)
   > Set looking_down(looking_down:double=GetAlphaFromReferenceTransformDelta(
     SkeletalMeshComponent=Get skeletal_mesh() -> (skeletal_mesh:Skeletal Mesh Component Object Reference)
     BoneName=head
     BaseName=spine_05
     DeltaComponentType=QuaternionTwist
     InputRange=(X=0,Y=-22)
     TwistAxis="0,0,1"
    )) -> (Output_Get:double)
   Is Valid(exec=@n10.then_1, InputObject=Get PlayerPawn() -> (PlayerPawn:Fort Player Pawn Athena Object Reference))
   > Set WPO Globals Function(
    Material=Get Mat Instance Dynamic()
    Overlay Material=Get Overlay Mat Instance Dynamic()
    Skeletal Mesh=Get SkelMesh()
    "Local Noise Transform  Location"=Get Local Noise Transform Location() -> (Local Noise Transform Location:Vector)
    Local Noise Offset=Get Local Noise Offset()
   )
   > Set WPO Regional Function!(
    execute
    Material:Material Instance Dynamic Object Reference=Get Mat Instance Dynamic()
    Overlay Material:Material Instance Dynamic Object Reference=Get Overlay Mat Instance Dynamic()
    Skeletal Mesh:Skeletal Mesh Component Object Reference=Get SkelMesh()
    Velocity Transform:Name=foot_r
    Velocity Interp Speed:double=Get Velocity Interp Speed()
    Collision Transform:Name=calf_r
    Collision Transform Axis:EAxis Enum
    Invert Collision Transform Axis:bool=false
    Delta Time X:double=@n9.DeltaTimeX
    Previous Velocity Transform Location:"Vector (by ref)"=Get R1 Velocity Transform Location()
    Previous Velocity Transform Velocity:"Vector (by ref)"=Get R1 Velocity Transform Velocity()
    Region:int=1
   ) -> (then)
   > Set WPO Regional Function(
    Material=Get Mat Instance Dynamic()
    Overlay Material=Get Overlay Mat Instance Dynamic()
    Skeletal Mesh=Get SkelMesh()
    Velocity Transform=foot_l
    Velocity Interp Speed=Get Velocity Interp Speed()
    Collision Transform=calf_l
    Invert Collision Transform Axis=true
    Delta Time X=@n9.DeltaTimeX
    Previous Velocity Transform Location=Get R2 Velocity Transform Location()
    Previous Velocity Transform Velocity=Get R2 Velocity Transform Velocity()
    Region=2
   )
   > n11 = Branch(Condition=Get bIsSkydiveFloating())
   > n12 = Set WPO Pho Wind Intensity(WPO Pho Wind Intensity=1)
   SetScalarParameterValue(execute=[@n13.then, @n12.then], self=Get Mat Instance Dynamic(), ParameterName=ClothSim_WindStrength, Value=Get WPO Pho Wind Intensity())
   n13 = Set WPO Pho Wind Intensity(execute=@n11.else, WPO Pho Wind Intensity=0)
   n14 = GetAnimInstance(self=GetOwnerSkeletalMeshComponent())
   Is Valid(exec=@n10.then_2, InputObject=@n14)
   > n15 = Branch(Condition=Get bIsInFrontEnd())
   > Set is_playing_character_idle(is_playing_character_idle:bool=Montage_IsActive!(
     self:Anim Instance Object Reference=@n14
     Montage:Anim Montage Object Reference=/Game/Animation/Game/MainPlayer/Menu/BR/CustomSkins/_s21/BariumF/Barium_Female_Idle_Montage
    ) -> (ReturnValue:bool)) -> (Output_Get:bool)
   Set is_playing_character_idle(execute=@n15.else, is_playing_character_idle=false)
   Set pantlegAlphaFrontend(execute=@n10.then_3, pantlegAlphaFrontend:double=SelectFloat(A=0.65, B=1, bPickA=Get bIsInFrontEnd())) -> (Output_Get:double)
  ]}
]
variables [
 custom_cape_equipped {type=boolean}
 is_playing_character_idle {type=boolean}
 Local Noise Offset {type=Vector}
 Local Noise Transform Location {type=Vector}
 looking_down {type=number}
 Mat Instance Dynamic {type=MaterialInstanceDynamic}
 Overlay Mat Instance Dynamic {type=MaterialInstanceDynamic}
 pantlegAlphaFrontend {type=number}
 PlayerPawn {type=FortPlayerPawnAthena}
 R1 Velocity Transform Location {type=Vector}
 R1 Velocity Transform Velocity {type=Vector}
 R2 Velocity Transform Location {type=Vector}
 R2 Velocity Transform Velocity {type=Vector}
 skeletal_mesh {type=SkeletalMeshComponent}
 SkelMesh {type=SkeletalMeshComponent}
 Velocity Interp Speed {type=number, default=30}
 WPO Pho Wind Intensity {type=number}
]
