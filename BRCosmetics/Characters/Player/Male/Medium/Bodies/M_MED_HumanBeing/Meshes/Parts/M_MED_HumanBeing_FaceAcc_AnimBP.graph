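# Animation Blueprint for the male medium HumanBeing face accessory mesh, adding physics-driven secondary motion to jaw, hair and glasses.
# Copy Pose From Mesh takes the attached parent's pose into a serial chain of angle-limited Anim Dynamics nodes.
# Jaw and 24 hair bones simulate in world space, glasses bone-relative.
# An emote-gated glasses transform node sits unconnected to the output, and the event graph has no logic.
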AnimBlueprint /BRCosmetics/Characters/Player/Male/Medium/Bodies/M_MED_HumanBeing/Meshes/Parts/M_MED_HumanBeing_FaceAcc_AnimBP
target_skeleton=&/M_MED_HumanBeing_FaceAcc_Skeleton
graphs [
 AnimGraph {nodes [
   "Transform (Modify) Bone - Bone: dyn_glasses_MSTR"(Alpha="To Float (Boolean)"(InBool=Get bIsPlayingEmote()), Translation="0.5,1.2,0", Rotation="0,35.000,0", Scale="1,1,1") {
    rotationMode=BMM_Additive
    rotationSpace=BCS_BoneSpace
    scaleSpace=BCS_BoneSpace
    translationMode=BMM_Additive
    translationSpace=BCS_BoneSpace
    alphaBoolBlend {blendOption=Linear}
    alphaScaleBiasClamp {bInterpResult=true}
    boneToModify {boneName=dyn_glasses_MSTR}
    rotation {yaw=35}
    translation=<0.5 1.2>
   }
   n1 = Anim Dynamics - Bone: dyn_jaw(ComponentPose=Local To Component(LocalPose=Copy Pose From Mesh() {bUseAttachedParent=true}), Alpha=1, ExternalForce="0,0,0") {
    simulationSpace=World
    alphaBoolBlend {blendOption=Linear}
    boundBone {boneName=dyn_jaw}
    physicsBodyDefinitions [{
      boundBone {boneName=dyn_jaw}
      boxExtents=<24 12 24>
      constraintSetup {angularLimitsMax=<0 0 15>, angularLimitsMin=<0 0 -3>, angularTarget=<0 0 0>}
      localJointOffset=<1 -0.3 0>
     }]
   }
   n2 = Anim Dynamics - Bone: dyn_hair_1(ComponentPose=@n1, Alpha=1, ExternalForce="0,0,0") {
    simulationSpace=World
    alphaBoolBlend {blendOption=Linear}
    boundBone {boneName=dyn_hair_1}
    physicsBodyDefinitions [{
      boundBone {boneName=dyn_hair_1}
      boxExtents=<18 12 12>
      constraintSetup {angularLimitsMax=<0 8 8>, angularLimitsMin=<0 -4 -8>, angularTarget=<1 0 0>}
      localJointOffset=<0.3 0 0>
     }]
   }
   n3 = Anim Dynamics - Bone: dyn_hair_2(ComponentPose=@n2, Alpha=1, ExternalForce="0,0,0") {
    simulationSpace=World
    alphaBoolBlend {blendOption=Linear}
    boundBone {boneName=dyn_hair_2}
    physicsBodyDefinitions [{
      boundBone {boneName=dyn_hair_2}
      boxExtents=<18 12 12>
      constraintSetup {angularLimitsMax=<0 8 8>, angularLimitsMin=<0 -4 -8>, angularTarget=<1 0 0>}
      localJointOffset=<0.3 0 0>
     }]
   }
   n4 = Anim Dynamics - Bone: dyn_hair_3(ComponentPose=@n3, Alpha=1, ExternalForce="0,0,0") {
    simulationSpace=World
    alphaBoolBlend {blendOption=Linear}
    boundBone {boneName=dyn_hair_3}
    physicsBodyDefinitions [{
      boundBone {boneName=dyn_hair_3}
      boxExtents=<18 12 12>
      constraintSetup {angularLimitsMax=<0 8 8>, angularLimitsMin=<0 -4 -8>, angularTarget=<1 0 0>}
      localJointOffset=<0.3 0 0>
     }]
   }
   n5 = Anim Dynamics - Bone: dyn_hair_4(ComponentPose=@n4, Alpha=1, ExternalForce="0,0,0") {
    simulationSpace=World
    alphaBoolBlend {blendOption=Linear}
    boundBone {boneName=dyn_hair_4}
    physicsBodyDefinitions [{
      boundBone {boneName=dyn_hair_4}
      boxExtents=<18 12 12>
      constraintSetup {angularLimitsMax=<0 8 8>, angularLimitsMin=<0 -4 -8>, angularTarget=<1 0 0>}
      localJointOffset=<0.3 0 0>
     }]
   }
   n6 = Anim Dynamics - Bone: dyn_hair_5(ComponentPose=@n5, Alpha=1, ExternalForce="0,0,0") {
    simulationSpace=World
    alphaBoolBlend {blendOption=Linear}
    boundBone {boneName=dyn_hair_5}
    physicsBodyDefinitions [{
      boundBone {boneName=dyn_hair_5}
      boxExtents=<18 12 12>
      constraintSetup {angularLimitsMax=<0 8 8>, angularLimitsMin=<0 -4 -8>, angularTarget=<1 0 0>}
      localJointOffset=<0.3 0 0>
     }]
   }
   n7 = Anim Dynamics - Bone: dyn_hair_6(ComponentPose=@n6, Alpha=1, ExternalForce="0,0,0") {
    simulationSpace=World
    alphaBoolBlend {blendOption=Linear}
    boundBone {boneName=dyn_hair_6}
    physicsBodyDefinitions [{
      boundBone {boneName=dyn_hair_6}
      boxExtents=<18 12 12>
      constraintSetup {angularLimitsMax=<0 8 8>, angularLimitsMin=<0 -4 -8>, angularTarget=<1 0 0>}
      localJointOffset=<0.3 0 0>
     }]
   }
   n8 = Anim Dynamics - Bone: dyn_hair_7(ComponentPose=@n7, Alpha=1, ExternalForce="0,0,0") {
    simulationSpace=World
    alphaBoolBlend {blendOption=Linear}
    boundBone {boneName=dyn_hair_7}
    physicsBodyDefinitions [{
      boundBone {boneName=dyn_hair_7}
      boxExtents=<18 12 12>
      constraintSetup {angularLimitsMax=<0 8 8>, angularLimitsMin=<0 -4 -8>, angularTarget=<1 0 0>}
      localJointOffset=<0.3 0 0>
     }]
   }
   n9 = Anim Dynamics - Bone: dyn_hair_8(ComponentPose=@n8, Alpha=1, ExternalForce="0,0,0") {
    simulationSpace=World
    alphaBoolBlend {blendOption=Linear}
    boundBone {boneName=dyn_hair_8}
    physicsBodyDefinitions [{
      boundBone {boneName=dyn_hair_8}
      boxExtents=<18 12 12>
      constraintSetup {angularLimitsMax=<0 8 8>, angularLimitsMin=<0 -4 -8>, angularTarget=<1 0 0>}
      localJointOffset=<0.3 0 0>
     }]
   }
   n10 = Anim Dynamics - Bone: dyn_hair_9(ComponentPose=@n9, Alpha=1, ExternalForce="0,0,0") {
    simulationSpace=World
    alphaBoolBlend {blendOption=Linear}
    boundBone {boneName=dyn_hair_9}
    physicsBodyDefinitions [{
      boundBone {boneName=dyn_hair_9}
      boxExtents=<18 12 12>
      constraintSetup {angularLimitsMax=<0 8 8>, angularLimitsMin=<0 -4 -8>, angularTarget=<1 0 0>}
      localJointOffset=<0.3 0 0>
     }]
   }
   n11 = Anim Dynamics - Bone: dyn_hair_10(ComponentPose=@n10, Alpha=1, ExternalForce="0,0,0") {
    simulationSpace=World
    alphaBoolBlend {blendOption=Linear}
    boundBone {boneName=dyn_hair_10}
    physicsBodyDefinitions [{
      boundBone {boneName=dyn_hair_10}
      boxExtents=<18 12 12>
      constraintSetup {angularLimitsMax=<0 8 8>, angularLimitsMin=<0 -4 -8>, angularTarget=<1 0 0>}
      localJointOffset=<0.3 0 0>
     }]
   }
   n12 = Anim Dynamics - Bone: dyn_hair_11(ComponentPose=@n11, Alpha=1, ExternalForce="0,0,0") {
    simulationSpace=World
    alphaBoolBlend {blendOption=Linear}
    boundBone {boneName=dyn_hair_11}
    physicsBodyDefinitions [{
      boundBone {boneName=dyn_hair_11}
      boxExtents=<18 12 12>
      constraintSetup {angularLimitsMax=<0 8 8>, angularLimitsMin=<0 -4 -8>, angularTarget=<1 0 0>}
      localJointOffset=<0.3 0 0>
     }]
   }
   n13 = Anim Dynamics - Bone: dyn_hair_12(ComponentPose=@n12, Alpha=1, ExternalForce="0,0,0") {
    simulationSpace=World
    alphaBoolBlend {blendOption=Linear}
    boundBone {boneName=dyn_hair_12}
    physicsBodyDefinitions [{
      boundBone {boneName=dyn_hair_12}
      boxExtents=<18 12 12>
      constraintSetup {angularLimitsMax=<0 8 8>, angularLimitsMin=<0 -4 -8>, angularTarget=<1 0 0>}
      localJointOffset=<0.3 0 0>
     }]
   }
   n14 = Anim Dynamics - Bone: dyn_hair_13(ComponentPose=@n13, Alpha=1, ExternalForce="0,0,0") {
    simulationSpace=World
    alphaBoolBlend {blendOption=Linear}
    boundBone {boneName=dyn_hair_13}
    physicsBodyDefinitions [{
      boundBone {boneName=dyn_hair_13}
      boxExtents=<18 12 12>
      constraintSetup {angularLimitsMax=<0 8 8>, angularLimitsMin=<0 -4 -8>, angularTarget=<1 0 0>}
      localJointOffset=<0.3 0 0>
     }]
   }
   n15 = Anim Dynamics - Bone: dyn_hair_14(ComponentPose=@n14, Alpha=1, ExternalForce="0,0,0") {
    simulationSpace=World
    alphaBoolBlend {blendOption=Linear}
    boundBone {boneName=dyn_hair_14}
    physicsBodyDefinitions [{
      boundBone {boneName=dyn_hair_14}
      boxExtents=<18 12 12>
      constraintSetup {angularLimitsMax=<0 8 8>, angularLimitsMin=<0 -4 -8>, angularTarget=<1 0 0>}
      localJointOffset=<0.3 0 0>
     }]
   }
   n16 = Anim Dynamics - Bone: dyn_hair_15(ComponentPose=@n15, Alpha=1, ExternalForce="0,0,0") {
    simulationSpace=World
    alphaBoolBlend {blendOption=Linear}
    boundBone {boneName=dyn_hair_15}
    physicsBodyDefinitions [{
      boundBone {boneName=dyn_hair_15}
      boxExtents=<18 12 12>
      constraintSetup {angularLimitsMax=<0 8 8>, angularLimitsMin=<0 -4 -8>, angularTarget=<1 0 0>}
      localJointOffset=<0.3 0 0>
     }]
   }
   n17 = Anim Dynamics - Bone: dyn_hair_16(ComponentPose=@n16, Alpha=1, ExternalForce="0,0,0") {
    simulationSpace=World
    alphaBoolBlend {blendOption=Linear}
    boundBone {boneName=dyn_hair_16}
    physicsBodyDefinitions [{
      boundBone {boneName=dyn_hair_16}
      boxExtents=<18 12 12>
      constraintSetup {angularLimitsMax=<0 8 8>, angularLimitsMin=<0 -4 -8>, angularTarget=<1 0 0>}
      localJointOffset=<0.3 0 0>
     }]
   }
   n18 = Anim Dynamics - Bone: dyn_hair_17(ComponentPose=@n17, Alpha=1, ExternalForce="0,0,0") {
    simulationSpace=World
    alphaBoolBlend {blendOption=Linear}
    boundBone {boneName=dyn_hair_17}
    physicsBodyDefinitions [{
      boundBone {boneName=dyn_hair_17}
      boxExtents=<18 12 12>
      constraintSetup {angularLimitsMax=<0 8 8>, angularLimitsMin=<0 -4 -8>, angularTarget=<1 0 0>}
      localJointOffset=<0.3 0 0>
     }]
   }
   n19 = Anim Dynamics - Bone: dyn_hair_18(ComponentPose=@n18, Alpha=1, ExternalForce="0,0,0") {
    simulationSpace=World
    alphaBoolBlend {blendOption=Linear}
    boundBone {boneName=dyn_hair_18}
    physicsBodyDefinitions [{
      boundBone {boneName=dyn_hair_18}
      boxExtents=<18 12 12>
      constraintSetup {angularLimitsMax=<0 8 8>, angularLimitsMin=<0 -4 -8>, angularTarget=<1 0 0>}
      localJointOffset=<0.3 0 0>
     }]
   }
   n20 = Anim Dynamics - Bone: dyn_hair_19(ComponentPose=@n19, Alpha=1, ExternalForce="0,0,0") {
    simulationSpace=World
    alphaBoolBlend {blendOption=Linear}
    boundBone {boneName=dyn_hair_19}
    physicsBodyDefinitions [{
      boundBone {boneName=dyn_hair_19}
      boxExtents=<18 12 12>
      constraintSetup {angularLimitsMax=<0 8 8>, angularLimitsMin=<0 -4 -8>, angularTarget=<1 0 0>}
      localJointOffset=<0.3 0 0>
     }]
   }
   n21 = Anim Dynamics - Bone: dyn_hair_20(ComponentPose=@n20, Alpha=1, ExternalForce="0,0,0") {
    simulationSpace=World
    alphaBoolBlend {blendOption=Linear}
    boundBone {boneName=dyn_hair_20}
    physicsBodyDefinitions [{
      boundBone {boneName=dyn_hair_20}
      boxExtents=<18 12 12>
      constraintSetup {angularLimitsMax=<0 8 8>, angularLimitsMin=<0 -4 -8>, angularTarget=<1 0 0>}
      localJointOffset=<0.3 0 0>
     }]
   }
   n22 = Anim Dynamics - Bone: dyn_hair_21(ComponentPose=@n21, Alpha=1, ExternalForce="0,0,0") {
    simulationSpace=World
    alphaBoolBlend {blendOption=Linear}
    boundBone {boneName=dyn_hair_21}
    physicsBodyDefinitions [{
      boundBone {boneName=dyn_hair_21}
      boxExtents=<18 12 12>
      constraintSetup {angularLimitsMax=<0 8 8>, angularLimitsMin=<0 -4 -8>, angularTarget=<1 0 0>}
      localJointOffset=<0.3 0 0>
     }]
   }
   n23 = Anim Dynamics - Bone: dyn_hair_22(ComponentPose=@n22, Alpha=1, ExternalForce="0,0,0") {
    simulationSpace=World
    alphaBoolBlend {blendOption=Linear}
    boundBone {boneName=dyn_hair_22}
    physicsBodyDefinitions [{
      boundBone {boneName=dyn_hair_22}
      boxExtents=<18 12 12>
      constraintSetup {angularLimitsMax=<0 8 8>, angularLimitsMin=<0 -4 -8>, angularTarget=<1 0 0>}
      localJointOffset=<0.3 0 0>
     }]
   }
   n24 = Anim Dynamics - Bone: dyn_hair_23(ComponentPose=@n23, Alpha=1, ExternalForce="0,0,0") {
    simulationSpace=World
    alphaBoolBlend {blendOption=Linear}
    boundBone {boneName=dyn_hair_23}
    physicsBodyDefinitions [{
      boundBone {boneName=dyn_hair_23}
      boxExtents=<18 12 12>
      constraintSetup {angularLimitsMax=<0 8 8>, angularLimitsMin=<0 -4 -8>, angularTarget=<1 0 0>}
      localJointOffset=<0.3 0 0>
     }]
   }
   n25 = Anim Dynamics - Bone: dyn_hair_24(ComponentPose=@n24, Alpha=1, ExternalForce="0,0,0") {
    simulationSpace=World
    alphaBoolBlend {blendOption=Linear}
    boundBone {boneName=dyn_hair_24}
    physicsBodyDefinitions [{
      boundBone {boneName=dyn_hair_24}
      boxExtents=<18 12 12>
      constraintSetup {angularLimitsMax=<0 8 8>, angularLimitsMin=<0 -4 -8>, angularTarget=<1 0 0>}
      localJointOffset=<0.3 0 0>
     }]
   }
   n26 = Anim Dynamics - Bone: dyn_glasses(ComponentPose=@n25, Alpha=1, ExternalForce="0,0,0") {
    simulationSpace=BoneRelative
    alphaBoolBlend {blendOption=Linear}
    boundBone {boneName=dyn_glasses}
    physicsBodyDefinitions [{
      boundBone {boneName=dyn_glasses}
      boxExtents=<5 8 20>
      constraintSetup {angularLimitsMax=<0 0 0.1>, angularLimitsMin=<0 0 -10>, angularTarget=<0 0 0>}
      localJointOffset=<0.3 0 0>
     }]
   }
   Output Pose(Result=Component To Local(ComponentPose=@n26)) {layerGroup=None}
  ]}
 EventGraph {nodes [Event BlueprintUpdateAnimation()]}
]
variables []
